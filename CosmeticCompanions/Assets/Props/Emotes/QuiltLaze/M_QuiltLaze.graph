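# Control Rig for the QuiltLaze emote props in CosmeticCompanions, a grindstone and an axe.
# Under a shared Root bone, the grindstone sits below a placement bone and the axe below Axe_Main.
# Connector names point to a modular rig with a Root module and an add-control module for each of the four child bones.
# The rig graph holds only an unlinked Begin Execution node, so no solve logic appears here.
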ControlRigBlueprint /CosmeticCompanions/Assets/Props/Emotes/QuiltLaze/M_QuiltLaze
graphs [
 RigVMFunctionLibrary {links [], nodes []}
 RigVMModel {nodes [RigUnit_BeginExecution = &RigUnit_BeginExecution()]}
]
hierarchy [
 Root {loc=<-2.2423 0 0>, rot=<0 0 0 1>}
 GringStone_Placement {parent=Root, loc=<-13.6693 0 0>, rot=<0 0 0 1>}
 GrindStone {parent=GringStone_Placement, loc=<-0.4784 0 21.9274>, rot=<0 0 0 1>}
 Axe_Main {parent=Root, loc=<19.2004 -0.0926 6.8524>, rot=<0 0 0 1>}
 Axe {parent=Axe_Main, loc=<0 0 0>, rot=<0 0 0 1>}
 Root/Root {type=Connector, loc=<0 0 0>, rot=<0 0 0 1>}
 Grindstone Placement/Add Control Primary {type=Connector, loc=<0 0 0>, rot=<0 0 0 1>}
 Grindstone Placement/Parent Control {type=Connector, loc=<0 0 0>, rot=<0 0 0 1>}
 Grindstone/Add Control Primary {type=Connector, loc=<0 0 0>, rot=<0 0 0 1>}
 Grindstone/Parent Control {type=Connector, loc=<0 0 0>, rot=<0 0 0 1>}
 Axe/Add Control Primary {type=Connector, loc=<0 0 0>, rot=<0 0 0 1>}
 Axe/Parent Control {type=Connector, loc=<0 0 0>, rot=<0 0 0 1>}
 Axe Main/Add Control Primary {type=Connector, loc=<0 0 0>, rot=<0 0 0 1>}
 Axe Main/Parent Control {type=Connector, loc=<0 0 0>, rot=<0 0 0 1>}
]
